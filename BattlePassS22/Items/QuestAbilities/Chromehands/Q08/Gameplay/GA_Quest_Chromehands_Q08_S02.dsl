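; Quest ability for Chromehands Q08 stage S02 that credits quest progress for qualifying targets in a triggering gameplay event.
; On activation it calls the parent, resets its counters, then gathers unique valid actors from the target data that match the Tag Container and are not on the owner's team.
; If any qualify, it sends their count as a complex custom stat event to the owner's quest manager, then ends the ability.
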(event Ability|EventActivateAbilityFromEvent
  (params (EventData :type Gameplay%20Event%20Data%20Structure%20%28by%20ref%29))
  (bind (_eventtag _instigator _target _optionalobject _optionalobject2 _contexthandle _instigatortags _targettags _eventmagnitude _targetdata) (Utilities|Struct|BreakGameplayEventData EventData))
  (|CallParentFunction EventData)
  (Variables|Default|SetValidTargets 0)
  (Utilities|Array|Clear (Variables|Default|GetActors%20Counted))
  (for _array_element (Ability|TargetData|GetAllActorsfromTargetData _targetdata)
    (Utilities|IsValid _array_element
      (:Is%20Valid
        (bind _returnvalue (GameplayTags|HasAnyMatchingGameplayTags(Message) _array_element (Variables|Default|GetTag%20Container)))
        (if (Math|Boolean|ANDBoolean _returnvalue (not (Team|OnSameTeam (Variables|GAVars|Data|GetOwningPlayerPawn) _array_element)) (not (Utilities|Array|ContainsItem (Variables|Default|GetActors%20Counted) _array_element)))
          (Utilities|Array|AddUnique (Variables|Default|GetActors%20Counted) _array_element)
          (Math|Integer|IncrementInt (Variables|Default|GetValidTargets))))))
  (if (> (Variables|Default|GetValidTargets) 0)
    (Quest|SendComplexCustomStatEvent (Variables|GAVars|Data|GetOwningPlayerQuestManager) 0 (Variables|Default|GetAdditional%20Source%20Tags) "(GameplayTags=)" (Variables|Default|GetValidTargets))
    (else
      _))
  (Ability|EndAbility))
